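# Masked material for mesh particles that cracks and erodes a prop surface, with Dusty Diner sign textures as defaults.
# A crack texture, mapped by UV or optionally world-aligned, alters color, specular and roughness and combines with scrolling-free noise against the particle Dynamic Parameter Erode to drive the opacity mask.
# The same erosion value tints dissolving edges and pushes vertices along their normals, an offset skipped on ES2 and ES3.1.
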Material /Figment_S08/Effects/Fort_Effects/World/Props/Dusty/M_ErodeMat_Custom
graph {
 properties {
  bUseMaterialAttributes=true
  bUsedWithMeshParticles=true
  blendMode=BLEND_Masked
  opacityMaskClipValue=0.01
  pixelDepthOffsetMode=PDOM_Legacy
  usedShadingModels=Default Lit
 }
 nodes [
  n1 = TextureSampleParameter2D() {parameterName=D, texture=/Game/Environments/Props/Signs/Textures/T_Dusty_Diner_Sign_D}
  n2 = TextureSampleParameter2D() {parameterName=S, samplerType=SAMPLERTYPE_Masks, texture=/Game/Environments/Props/Signs/Textures/T_Dusty_Diner_Sign_S}
  n3 = SetMaterialAttributes!(
   MaterialAttributes
   Base Color=@n1.RGB
   Metallic=@n2.G
   Specular=@n2.G
   Roughness=@n2.G
   Normal=TextureSampleParameter2D() {parameterName=N, samplerType=SAMPLERTYPE_Normal, texture=/Game/Environments/Props/Signs/Textures/T_Dusty_Diner_Sign_N}.RGB
  ) {attributeSetTypes [
    #1
    #2
    #3
    #4
    #5
   ]}
  n4 = GetMaterialAttributes(None=@n3) {attributeGetTypes [#1, #3, #4]}
  n5 = TextureCoordinate()
  n6 = TextureObject() {texture=/Game/Global/Textures/Default/Masks_and_Gradients/Cracks_02}
  n7 = Constant() {r=350}
  n8 = WorldPosition() {worldPositionShaderOffset=WPT_ExcludeAllShaderOffsets}
  n9 = Multiply(
   A=StaticSwitch(
    True=TextureSample(UVs=@n5, Tex=@n6) {texture=&Cracks_02}.RGB
    False=MaterialFunctionCall(TextureObject=@n6, TextureSize=@n7, WorldPosition=@n8.XYZ) {materialFunction=&WorldAlignedTexture}.XYZ Texture
    Value=StaticBoolParameter() {defaultValue=true, parameterName=Use0UVasCrack}
   ) {desc=use uv 2 for destruction - save 40 instructions}
   B=ScalarParameter() {defaultValue=1, parameterName=CrackSize}
  )
  n10 = LinearInterpolate(
   A=@n4.Base Color
   B=VectorParameter() {parameterName=CrackColor, defaultValue=<rgba 0.04 0.04 0.04 0>}.RGB
   Alpha=ComponentMask(None=@n9) {r=true}
  ) {constB=0.15}
  n11 = LinearInterpolate(A=@n4.Specular, Alpha=ComponentMask(None=@n9) {b=true}) {constB=0.1}
  n12 = LinearInterpolate(A=@n4.Roughness, Alpha=ComponentMask(None=@n9) {b=true})
  n13 = Multiply(A=ComponentMask(None=@n9) {b=true}, B=ScalarParameter() {defaultValue=0.5, parameterName=CrackAffectMask}) {constB=0.25}
  n14 = TextureSample(UVs=Multiply(A=TextureCoordinate(), B=ScalarParameter() {defaultValue=2, parameterName=NoiseUV})) {samplerType=SAMPLERTYPE_LinearColor, texture=/Game/Characters/Player/Male/Medium/Bodies/M_MED_SRDrift/Textures/T_ElectricNoise}
  n15 = ScalarParameter() {defaultValue=1, parameterName=NoiseAffectMask}
  n16 = Subtract(
   A=Add(A=Subtract(B=@n13) {constA=0.5}, B=Multiply(A=@n14.R, B=@n15))
   B=DynamicParameter() {defaultValue=<rgba 0 1 1 1>, paramNames [Erode,Param2,Param3,Param4]}.Erode
  )
  n17 = GetMaterialAttributes(None=SetMaterialAttributes!(
    MaterialAttributes=SetMaterialAttributes!(MaterialAttributes=@n4.MaterialAttributes, Base Color=@n10, Specular=@n11, Roughness=@n12) {attributeSetTypes [#1, #3, #4]}
    Opacity Mask=@n16
   ) {attributeSetTypes [#6]}) {attributeGetTypes [#1, #3, #4]}
  n18 = Multiply(A=@n16) {constB=-1}
  n19 = Saturate(None=Add(A=@n18) {constB=0.5})
  n20 = SetMaterialAttributes(
   MaterialAttributes=@n17.MaterialAttributes
   Base Color=LinearInterpolate(A=@n17.Base Color, B=VectorParameter() {parameterName=ErodeColor, defaultValue=<rgba 0.04 0.04 0.04 0>}.RGB, Alpha=@n19) {constB=0.1}
   Specular=LinearInterpolate(A=@n17.Specular, Alpha=@n19) {constB=0.15}
   Roughness=LinearInterpolate(A=@n17.Roughness, Alpha=@n19)
  ) {attributeSetTypes [#1, #3, #4]}
  n21 = VertexNormalWS()
  n22 = Saturate(None=@n18)
  n23 = SetMaterialAttributes!(
   MaterialAttributes=@n20
   World Position Offset=Multiply(A=Multiply(A=@n21, B=@n22), B=ScalarParameter() {defaultValue=50, parameterName=WpoStrength}) {constB=100}
  ) {attributeSetTypes [#7]}
  FeatureLevelSwitch(Default=@n23, ES2=@n20, ES3_1=@n20, SM4_REMOVED=@n23)
  MaterialFunctionCall() {materialFunction=&DebugTimeSine}
  Multiply(B=@n1.RGB)
  StaticBool()
 ]
}
